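; Visualization actor for a blink ability, showing a sphere and an arrow that yaw to face a target Fort pawn.
; Setting the replicated target pawn activates tracking; its OnRep starts a self-re-arming 0.1-second timer that updates rotation, while Tick is enabled but empty.
; The timer loop ends when the pawn is invalid or tracking is off.
; The TrackingActive OnRep swaps sphere material and arrow color between active and locked looks.
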(event EventBeginPlay)

(event Collision|EventActorBeginOverlap
  (params (OtherActor :type Actor%20Object%20Reference)))

(event EventTick
  (params (DeltaSeconds :type Float%20%28single-precision%29)))

(event Custom|OnUpdateRotation
  (bind _targetpawn (Variables|Default|GetTargetPawn))
  (if (and (Utilities|IsValid _targetpawn :!class /Script/Engine.KismetSystemLibrary) (Variables|Default|GetTrackingActive))
    (Transformation|SetActorRotation 0 (Math|Rotator|MakeRotator 0.0 0.0 (Math|Vector|GetYawPitchfromVector (- (Transformation|GetActorLocation _targetpawn) (Transformation|GetActorLocation)))) false)
    (Utilities|Time|SetTimerbyEvent (EventDispatchers|CreateEvent :!event Custom|OnUpdateRotation) 0.1)))

(fn OnRep_TargetPawn
  (Variables|Default|SetTrackingActive true)
  (Actor|Tick|SetActorTickEnabled true)
  (Utilities|Time|SetTimerbyEvent (EventDispatchers|CreateEvent :!event OnUpdateRotation) 0.1))

(fn OnRep_TrackingActive
  (Rendering|Material|SetMaterial (Variables|Default|GetSphere) 0 (Selection|SelectMaterialInstance (Variables|Default|GetTrackingActiveMaterial) (Variables|Default|GetTrackingLockedMaterial) (Variables|Default|GetTrackingActive)))
  (Components|Arrow|SetArrowColor(LinearColor) (Variables|Default|GetArrow) (Math|Color|SelectColor (Variables|Default|GetTrackingActiveArrowColor) (Variables|Default|GetTrackingLockedArrowColor) (Variables|Default|GetTrackingActive))))

(fn SetTargetPawn
  (params (NewTargetPawn :type Fort%20Pawn%20Object%20Reference))
  (Variables|Default|SetTargetPawn NewTargetPawn)
  (Variables|Default|SetTrackingActive true)
  (Actor|Tick|SetActorTickEnabled true))

(fn SetTrackingActive
  (params (TrackingActive :type Boolean))
  (Variables|Default|SetTrackingActive TrackingActive))

(fn ConstructionScript)
