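; Fire pillar hazard actor for CoffeePuff that shows an anticipation effect and sound on spawn, then explodes when Detonate is called.
; Detonate capsule-traces at HitCollider, skipping the instigator and spawners, and applies separate scalable damage for pawns, vehicles and buildings.
; Pawns and vehicles also get knocked away from the collider.
; Setting the replicated ActivatePillar flag drives the burst Niagara, camera shake and detonation sound through its OnRep.
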(event EventBeginPlay
  (Networking|SwitchHasAuthority
    (:Authority
      (Utilities|FlowControl|DoOnce false
        (:Completed
          (Variables|Default|SetActivatePillar false)
          (CallFunction|SpawnAnticipationVFX))))))

(event Custom|SpawnAnticipationVFX
  (bind _ns_coffeepuff_firepillar_anticipation (Variables|Default|GetNS_CoffeePuff_FirePillar_Anticipation))
  (Networking|CanExecuteCosmeticEvents
    (:True
      (Components|Activation|Deactivate _ns_coffeepuff_firepillar_anticipation)
      (Niagara|ResetSystem _ns_coffeepuff_firepillar_anticipation)
      (Components|Activation|SetActive _ns_coffeepuff_firepillar_anticipation true true)
      (Audio|SpawnSoundatLocation (Variables|Default|GetAnticipationSound) (Transformation|GetActorLocation)))))

(fn Detonate
  (Utilities|Array|Add (Variables|Default|GetActorsToIgnore) (Variables|Actor|GetInstigator))
  (bind _outhits (Collision|MultiCapsuleTraceForObjects (Transformation|GetWorldLocation (Variables|Default|GetHitCollider)) (+ (Transformation|GetWorldLocation (Variables|Default|GetHitCollider)) "0.000000,0.000000,1.000000") (Components|Capsule|GetScaledCapsuleRadius (Variables|Default|GetHitCollider)) (Components|Capsule|GetScaledCapsuleHalfHeight (Variables|Default|GetHitCollider)) (Variables|Default|GetAffectedObjectTypes) false (Variables|Default|GetActorsToIgnore)))
  (for _array_element _outhits
    (Utilities|Casting|CastToBGA_CoffeePuff_Spawner (.HitActor (Collision|BreakHitResult _array_element))
      (:CastFailed
        (bind _output_get (Variables|Default|SetCurrentDamageActor (.HitActor (Collision|BreakHitResult _array_element))))
        (bind _damage (CallFunction|GetDamageAmount _output_get))
        (if (> _damage 0)
          (Environment|ApplyGlobalEnvironmentDamagetoActor (Variables|Default|GetCurrentDamageActor) _damage)))))
  (Variables|Default|SetActivatePillar true))

(fn GetDamageAmount
  (params (Target :type Actor%20Object%20Reference))
  (bind _asfort_player_pawn (Utilities|Casting|CastToFortPlayerPawn Target)
    (:then
      (LocalVariables|SetTotalDamage (ScalableFloat|GetValueatLevel (Variables|Default|GetDamagePawn)))
      (CallFunction|KnockbackPawn _asfort_player_pawn)
      (GameplayCue|ExecuteGameplayCueOnActor(Burst) _asfort_player_pawn (Variables|Default|GetImpactGC) (Ability|GameplayCue|MakeGameplayCueParameters 0.0)))
    (:CastFailed
      (bind _asfort_athena_vehicle (Utilities|Casting|CastToFortAthenaVehicle Target)
        (:then
          (LocalVariables|SetTotalDamage (ScalableFloat|GetValueatLevel (Variables|Default|GetDamageVehicle)))
          (CallFunction|KnockbackVehicle _asfort_athena_vehicle))
        (:CastFailed
          (Utilities|Casting|CastToBuildingActor Target)
          (LocalVariables|SetTotalDamage (ScalableFloat|GetValueatLevel (Variables|Default|GetDamageBuildings)))))))
  (return (LocalVariables|GetTotalDamage)))

(fn KnockbackPawn
  (params (Pawn :type Fort%20Pawn%20Object%20Reference))
  (bind (_returnvalue_x _returnvalue_y) (Math|Vector|BreakVector (Math|Vector|Normalize (- (Transformation|GetActorLocation Pawn) (Transformation|GetWorldLocation (Variables|Default|GetHitCollider))))))
  (Character|LaunchCharacter Pawn (Math|Vector|MakeVector (* _returnvalue_x (ScalableFloat|GetValueatLevel (Variables|Default|GetKnockbackPawnVelocityXY))) (* _returnvalue_y (ScalableFloat|GetValueatLevel (Variables|Default|GetKnockbackPawnVelocityXY))) (ScalableFloat|GetValueatLevel (Variables|Default|GetKnockbackPawnVelocityZ))) true true))

(fn KnockbackVehicle
  (params (Vehicle :type Fort%20Athena%20Vehicle%20Object%20Reference))
  (bind (_returnvalue_x _returnvalue_y) (Math|Vector|BreakVector (Math|Vector|Normalize (- (Transformation|GetActorLocation Vehicle) (Transformation|GetWorldLocation (Variables|Default|GetHitCollider))))))
  (Audio|BroadcastAbilityImpactatLocation 0 Vehicle (Math|Vector|MakeVector (* _returnvalue_x (ScalableFloat|GetValueatLevel (Variables|Default|GetKnockbackVehicleVelocityXY))) (* _returnvalue_y (ScalableFloat|GetValueatLevel (Variables|Default|GetKnockbackVehicleVelocityXY))) (ScalableFloat|GetValueatLevel (Variables|Default|GetKnockbackVehicleVelocityZ))) 777.0 "None" false))

(fn OnRep_ActivatePillar
  (bind _ns_coffeepuff_firepillar_anticipation (Variables|Default|GetNS_CoffeePuff_FirePillar_Anticipation))
  (if (Variables|Default|GetActivatePillar)
    (Components|Activation|Activate (Variables|Default|GetNS_CoffeePuff_FirePillar_Burst) false)
    (Camera|PlayWorldCameraShake "/VividRazor_CoffeePuff/Assets/VFX/B_TEMP_CoffeePuff_ImpactShake.B_TEMP_CoffeePuff_ImpactShake_C" (+ (Transformation|GetActorLocation self) "0.000000,0.000000,-15.000000") 5e+02 1e+03)
    (Niagara|NiagaraSetBoolArray _ns_coffeepuff_firepillar_anticipation "ExplodeArray" (Utilities|Array|MakeArray true))
    (Niagara|NiagaraSetPositionArray _ns_coffeepuff_firepillar_anticipation "SpawnRateList" (Utilities|Array|MakeArray))
    (Niagara|NiagaraSetFloatArray _ns_coffeepuff_firepillar_anticipation "SpawnRateStartTime" (Utilities|Array|MakeArray))
    (Niagara|NiagaraSetPositionArray (Variables|Default|GetNS_CoffeePuff_FirePillar_Burst) "SpawnArray" (Utilities|Array|MakeArray (+ (Transformation|GetActorLocation self) "0.000000,0.000000,-15.000000")))
    (Audio|SpawnSoundatLocation (Variables|Default|GetDetonationSound) (Transformation|GetActorLocation))))

(fn ConstructionScript)
