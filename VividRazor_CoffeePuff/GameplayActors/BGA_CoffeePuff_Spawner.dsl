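; BGA_CoffeePuff_Spawner is a gong that summons the CoffeePuff AI boss when interacted with or hit by melee, healing back damage taken.
; Activation shakes nearby cameras, spawns a fire pillar, then after delays detonates it and requests the boss through the Living World point provider.
; The spawned boss can target the gong ringer, and its death updates encounter state.
; OnRep functions drive gong sound, material-based animation and state notifications.
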(event EventBeginPlay
  (Class|FortAsyncActionCurrentPlaylistReady|PlaylistDataReadyAsync
    (:OnReady
      (bind _returnvalue (Class|FortAthenaLivingWorldManager|RegisterDensityReservoir (Class|FortAthenaLivingWorldManager|GetLivingWorldManager) self 1.0 1.5e+04))
      (if _returnvalue
        (Utilities|IsValid (Variables|Default|GetPointProvider)
          (:Is%20Valid
            (Variables|Default|SetUserCanInteract true)
            (Variables|Default|SetBoss%20Spawn%20Location (Transformation|GetActorLocation (Variables|Default|GetPointProvider)))))))))

(event Custom|ActivateSpawner
  (bind _returnvalue (Class|FortAthenaLivingWorldManager|GetLivingWorldManager))
  (bind _pointprovider (Variables|Default|GetPointProvider))
  (Variables|Default|SetEncounter%20State "NewEnumerator4")
  (CallFunction|TriggerCameraShake)
  (CallFunction|KillthePillarFX)
  (Utilities|FlowControl|Delay (ScalableFloat|GetValueatLevel (Variables|Default|GetSpawnPillarDelay)))
  (bind _returnvalue_2 (Building|SpawnBuildingGameplayActor "/VividRazor_CoffeePuff/GameplayActors/BGA_CoffeePuff_SpawnPillar.BGA_CoffeePuff_SpawnPillar_C" (Math|Transform|MakeTransform (Transformation|GetActorLocation (Variables|Default|GetPointProvider))) self true false))
  (bind _asbga_coffee_puff_spawn_pillar (Utilities|Casting|CastToBGA_CoffeePuff_SpawnPillar _returnvalue_2))
  (Variables|Default|SetSpawnPillar _asbga_coffee_puff_spawn_pillar)
  (Utilities|FlowControl|Delay (ScalableFloat|GetValueatLevel (Variables|Default|GetSpawnPillarDetonationDelay)))
  (Class|FortAthenaLivingWorldPointProviderInterface|EnablePointProvider_DEPRECATED _pointprovider)
  (EventDispatchers|BindEventtoOnActorSpawnedResult _pointprovider (EventDispatchers|CreateEvent :!event Custom|OnActorSpawned))
  (Class|FortAthenaLivingWorldManager|UnregisterDensityReservoir _returnvalue self)
  (Class|FortAthenaLivingWorldManager|RequestEventGeneration _returnvalue)
  (Class|BGACoffeePuffFirePillar|Detonate (Variables|Default|GetSpawnPillar))
  (Utilities|FlowControl|Delay (ScalableFloat|GetValueatLevel (Variables|Default|GetPostSpawnAnnouncerDelay)))
  (if (Utilities|Enum|Equal(Enum) (Variables|Default|GetEncounter%20State) "NewEnumerator4")
    (Variables|Default|SetEncounter%20State "NewEnumerator1")))

(event Custom|TrySpawnAI
  (Utilities|FlowControl|DoOnce false
    (:Completed
      (CallFunction|ActivateSpawner))))

(event Interaction|EventBlueprintOnInteract
  (params (InteractingPawn :type Fort%20Pawn%20Object%20Reference) (InteractionBeingAttempted :type EInteractionBeingAttempted%20Enum :default "FirstInteraction"))
  (Variables|Default|SetInteractingPawn InteractingPawn)
  (Variables|Default|SetInteractionSuccessful true)
  (Utilities|FlowControl|DoOnce false
    (:Completed
      (CallFunction|ActivateSpawner))))

(event Custom|KillThePillarFX
  (Networking|CanExecuteCosmeticEvents
    (:True
      (Components|Activation|Deactivate (Variables|Default|GetNS_CoffeePuff_LightPillar)))))

(event Custom|PlayImpactSound
  (Networking|CanExecuteCosmeticEvents
    (:True
      (Audio|SpawnSoundAttached (Variables|Default|GetImpactSoundCue) (Variables|Default|GetSM_Dojo_Gong_A)))))

(event Custom|OnActorSpawned
  (params (SpawnedActor :type Actor%20Object%20Reference) (bResult :type Boolean))
  (bind _asbp_player_pawn_non_participant_vivid_razor_coffee_puff (Utilities|Casting|CastToBP_PlayerPawn_NonParticipant_VividRazor_CoffeePuff SpawnedActor))
  (Class|BPPlayerPawnNonParticipantVividRazorCoffeePuff|SetSpawnLocation _asbp_player_pawn_non_participant_vivid_razor_coffee_puff (Transformation|GetActorLocation (Variables|Default|GetPointProvider)))
  (EventDispatchers|BindEventtoOnFortPlayerDied _asbp_player_pawn_non_participant_vivid_razor_coffee_puff (delegate Custom|CoffeePuffDied))
  (Components|Activation|Deactivate (Variables|Default|GetCoffeePuff_Spawner_SpecialActor))
  (if (> (ScalableFloat|GetValueatLevel (Variables|Default|GetUseGongRingerAsInitialTarget)) 0)
    (Utilities|FlowControl|Delay (ScalableFloat|GetValueatLevel (Variables|Default|GetInitialTargetAssignmentDelay)))
    (bind _asfort_athena_aibot_controller (Utilities|Casting|CastToFortAthenaAIBotController (Pawn|GetController _asbp_player_pawn_non_participant_vivid_razor_coffee_puff)))
    (Class|FortAthenaAIBotController|AddTargetInfo _asfort_athena_aibot_controller (Variables|Default|GetInteractingPawn) false "Threat_LKP" (ScalableFloat|GetValueatLevel (Variables|Default|GetInitialTargetForgetTime)) (ScalableFloat|GetValueatLevel (Variables|Default|GetInitialTargetForgetDistance)))))

(event Custom|CoffeePuffDied
  (params (FortPlayerPawnAthena :type Fort%20Player%20Pawn%20Athena%20Object%20Reference))
  (Variables|Default|SetEncounter%20State "NewEnumerator3"))

(event EventOnDamageServer
  (params (Damage :type Float%20%28single-precision%29) (DamageTags :type Gameplay%20Tag%20Container%20Structure%20%28by%20ref%29) (Momentum :type Vector :default "0, 0, 0") (HitInfo :type Hit%20Result%20Structure%20%28by%20ref%29) (InstigatedBy :type Controller%20Object%20Reference) (DamageCauser :type Actor%20Object%20Reference) (EffectContext :type Gameplay%20Effect%20Context%20Handle%20Structure))
  (bind _self self)
  (bind _returnvalue (Ability|GetFortAbilitySystemComponent _self))
  (Utilities|Casting|CastToBP_PlayerPawn_NonParticipant_VividRazor_CoffeePuff (Pawn|GetControlledPawn InstigatedBy)
    (:then
      (if (> (ScalableFloat|GetValueatLevel (Variables|Default|GetBossIsAllowedToDestroyGong)) 0)
        _
        (else
          _)))
    (:CastFailed
      _))
  (bind _returnvalue_2 (Ability|GameplayEffect|AssignTagSetbyCallerMagnitude (GameplayEffects|MakeOutgoingSpec :self _returnvalue :GameplayEffectClass "/VividRazor_CoffeePuff/Abilities/GE_CoffeePuff_Spawner_InstantHeal.GE_CoffeePuff_Spawner_InstantHeal_C" :Level 1.0) "(TagName=\"SetByCaller.Healing\")" Damage))
  (GameplayEffects|ApplyGameplayEffectSpecToTarget _returnvalue _returnvalue_2 _returnvalue)
  (if (and (GameplayTags|HasAnyTags DamageTags (GameplayTags|MakeLiteralGameplayTagContainer "(GameplayTags=((TagName=\"Weapon.Melee\"),(TagName=\"Gameplay.Action.Attacking.Melee\")))") true) (not (Variables|Default|GetInteractionSuccessful)))
    (bind _asfort_pawn (Utilities|Casting|CastToFortPawn (Pawn|GetControlledPawn InstigatedBy)))
    (Variables|Default|SetInteractingPawn _asfort_pawn)
    (Variables|Default|SetInteractionSuccessful true)
    (CallFunction|ActivateSpawner)))

(fn BlueprintCanInteract
  (params (InteractingPawn :type Fort%20Pawn%20Object%20Reference) (InteractionBeingAttempted :type EInteractionBeingAttempted%20Enum :default "FirstInteraction") (InteractionType :type TInteractionType%20Enum :default "IT_NoInteraction"))
  (return (and (Variables|Default|GetUserCanInteract) (not (Variables|Default|GetInteractionSuccessful)))))

(fn BlueprintGetInteractionString
  (params (InteractingPawn :type Fort%20Pawn%20Object%20Reference) (InteractionBeingAttempted :type EInteractionBeingAttempted%20Enum :default "FirstInteraction"))
  (return (Variables|GetBuildingActorInteractionText)))

(fn CleanUpHitAnimation
  (Rendering|LOD|SetEvaluateWorldPositionOffset (Variables|Default|GetSM_Dojo_Gong_A) false :!class /Script/Engine.StaticMeshComponent))

(fn OnRep_EncounterState
  (Default|CallOnEncounterStateChanged (Variables|Default|GetEncounter%20State)))

(fn OnRep_InteractionSuccessful
  (CallFunction|PlayImpactSound)
  (CallFunction|PlayHitAnimation)
  (Audio|Components|Audio|FadeOut (Variables|/Game/Sounds/Fort_Audio_Blueprints/PropAmbientAudioComponent/FortPropAudioComponent.FortPropAudioComponent_C|GetLoopAudioComponent (Variables|Default|GetAmbientAudio)) 0.25 0.0 "Linear" :!class /Script/Engine.AudioComponent))

(fn OnRep_UserCanInteract
  (Variables|Interaction|SetbAllowInteract true))

(fn PlayHitAnimation
  (bind _sm_dojo_gong_a (Variables|Default|GetSM_Dojo_Gong_A))
  (Rendering|LOD|SetEvaluateWorldPositionOffset (Variables|Default|GetSM_Dojo_Gong_A) true :!class /Script/Engine.StaticMeshComponent)
  (Rendering|Material|SetScalarParameterValueonMaterials _sm_dojo_gong_a "IsForward" 0.0)
  (Rendering|Material|SetScalarParameterValueonMaterials _sm_dojo_gong_a "StartTime" (Utilities|Time|GetTimeSeconds))
  (Utilities|Time|ClearandInvalidateTimerbyHandle (Variables|Default|GetAnimationTimerHandle))
  (bind _returnvalue (Utilities|Time|SetTimerbyEvent (EventDispatchers|CreateEvent :!event CleanUpHitAnimation) 5.0))
  (Variables|Default|SetAnimationTimerHandle _returnvalue))

(fn SetPointProvider
  (params (Point%20Provider :type Fort%20Athena%20Patrol%20Path%20Point%20Provider%20Object%20Reference))
  (Variables|Default|SetPointProvider Point%20Provider))

(fn TriggerCameraShake
  (if (> (ScalableFloat|GetValueatLevel (Variables|Default|GetCameraShakeEnabled)) 0)
    (bind (_outactors _returnvalue) (Collision|SphereOverlapActors (Transformation|GetWorldLocation (Variables|Default|GetCameraShakeArea)) (Components|Sphere|GetScaledSphereRadius (Variables|Default|GetCameraShakeArea)) (Utilities|Array|MakeArray "ObjectTypeQuery3") "/Script/FortniteGame.FortPlayerPawn"))
    (if _returnvalue
      (for _array_element _outactors
        (bind _asfort_player_pawn (Utilities|Casting|CastToFortPlayerPawn _array_element))
        (if (Pawn|IsPlayerControlled _asfort_player_pawn)
          (bind _returnvalue_2 (Fortnite|GetFortPlayerControllerfromActor _asfort_player_pawn))
          (Game|Feedback|ClientStartCameraShake :self _returnvalue_2 :Shake "/VividRazor_CoffeePuff/GameplayActors/CoffeePuff_CameraShake_Spawner_Activation.CoffeePuff_CameraShake_Spawner_Activation_C" :Scale 1.0 :PlaySpace "CameraLocal"))))))

(fn ConstructionScript
  (Effects|Components|ParticleSystem|SetBoolParameter (Variables|Default|GetNS_CoffeePuff_LightPillar) "IsUnderground" (Variables|Default|GetPreventLightPillarPokeOutOfGround)))
