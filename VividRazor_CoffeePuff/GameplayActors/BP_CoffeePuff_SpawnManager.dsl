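; Spawn manager for CoffeePuff that activates its referenced location managers at match start and designates one as the boss location.
; On dedicated servers, it waits for playlist data, then checks that all managers are valid, with delayed retries capped by a hotfix value.
; The boss location is random among boss-enabled ones, with a hotfix index override.
; The construction script colors editor-only text to show whether references are set.
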(event EventBeginPlay
  (bind _randomlyselectedbossspawnlocationindex (Variables|Transient|GetRandomlySelectedBossSpawnLocationIndex))
  ;[0 0.15 0.07836 0.5] BeginPlay: Find the location managers in the world. Activate all of them while selecting one to be a boss location.
  (|CallParentFunction)
  (if (Networking|IsDedicatedServer)
    (Class|FortAsyncActionCurrentPlaylistReady|PlaylistDataReadyAsync
      (:OnReady
        ;;[0.1196 0.03831 0.15 0.5] Wait until all the location manager actors are loaded in.
        (Variables|Transient|SetbInvalidLocation false)
        (bind _array_element (Utilities|Array|ForEachLoopwithBreak (Variables|Managers|GetLocationManagers))
          (:LoopBody
            (Utilities|IsValid _array_element
              (:Is%20Not%20Valid
                (Variables|Transient|SetbInvalidLocation true))))
          (:Completed
            _))
        (if (Variables|Transient|GetbInvalidLocation)
          (Utilities|FlowControl|Delay 0.25)
          (bind _result (Math|Integer|IncrementInt (Variables|Transient|GetFindLocationManagersAttemptCount))
            (:%20%20
              (if (> _result (ScalableFloat|GetCachedScalableFloatValue (Variables|Hotfix|GetHotfix_FindLocationManagersMaxAttemptCount)))
                _
                (else
                  _))
              (Fortnite|FortShippingLog self "BeginPlay: Failed to find all location managers." true))))
        ;;[0 0.1479 0.15 0.5] Randomly select a location for the boss.
        (Fortnite|FortShippingLog self (Utilities|String|Append "BeginPlay: All location managers found successfully at attempt: " (Utilities|String|ToString(Integer) (Variables|Transient|GetFindLocationManagersAttemptCount))))
        (bind _returnvalue (CallFunction|RandomlySelectBossSpawnLocationIndex))
        (bind _output_get (Variables|Transient|SetRandomlySelectedBossSpawnLocationIndex _returnvalue))
        (if (Utilities|Array|IsValidIndex (Variables|Managers|GetLocationManagers) _output_get)
          ;;[0.06032 0.02261 0.15 0.5] Activate all the location managers.
          (for (_array_element_2 _array_index) (Variables|Managers|GetLocationManagers)
            (Utilities|IsValid _array_element_2
              (:Is%20Valid
                (Fortnite|FortShippingLog self (Utilities|String|Append "BeginPlay: Activating location: " (Utilities|GetDisplayName _array_element_2) " as Boss Location = " (Utilities|String|ToString(Boolean) (== _array_index _randomlyselectedbossspawnlocationindex))))
                (Class|BCoffeePuffLocationManager|ActivateLocation _array_element_2 (== _array_index _randomlyselectedbossspawnlocationindex)))))
          (else
            (Fortnite|FortShippingLog self (Utilities|String|Append "BeginPlay: RandomlySelectedIndex value is not a valid index: " (Utilities|String|ToString(Integer) _output_get) " ; LocationManagers array length: " (Utilities|String|ToString(Integer) (Utilities|Array|Length (Variables|Managers|GetLocationManagers)))) true)))))))

(fn RandomlySelectBossSpawnLocationIndex
  (bind _bossspawningenabledlocationindexes (LocalVariables|GetBossSpawningEnabledLocationIndexes))
  ;[0.03674 0.15 0.1442 0.5] Randomly select one of the boss spawning enabled locations.
  (Utilities|FlowControl|Sequence
    (:then_0
      (for (_array_element _array_index) (Variables|Managers|GetLocationManagers)
        (Utilities|IsValid _array_element
          (:Is%20Valid
            (if (Class|BCoffeePuffLocationManager|IsBossSpawningEnabledatLocation _array_element)
              (Utilities|Array|Add (LocalVariables|GetBossSpawningEnabledLocationIndexes) _array_index)))))
      (if (> (Utilities|Array|Length _bossspawningenabledlocationindexes) 0)
        (bind _returnvalue (Math|Random|RandomIntegerinRange 0 (Utilities|Array|LastIndex _bossspawningenabledlocationindexes)))
        (return (select (>= (ScalableFloat|GetValueasInteger (Variables|Hotfix|GetHotfix_ForceBossLocationIndex)) 0) (ScalableFloat|GetValueasInteger (Variables|Hotfix|GetHotfix_ForceBossLocationIndex)) (Utilities|Array|Get(aref) _bossspawningenabledlocationindexes _returnvalue)))))
    (:then_1
      (return -1))))

(fn ConstructionScript
  (bind _editoronly_feedback (Variables|Default|GetEDITORONLY_Feedback))
  (bind _editoronly_feedback_2 (Variables|Default|GetEDITORONLY_Feedback))
  ;[0.08421 0.03174 0.15 0.5] Update preview.
  (|CallParentFunction)
  (if (> (Utilities|Array|Length (Variables|Managers|GetLocationManagers)) 0)
    (bind _array_element (Utilities|Array|ForEachLoopwithBreak (Variables|Managers|GetLocationManagers))
      (:LoopBody
        (Utilities|IsValid _array_element
          (:Is%20Not%20Valid
            (bind _output_get (LocalVariables|SetbLocationReferenceMissing true)))))
      (:Completed
        _))
    (if (not _output_get)
      (Rendering|Components|TextRender|SetText _editoronly_feedback_2 (Utilities|Text|MakeLiteralText "INVTEXT(\"References set!\")"))
      (Rendering|Components|TextRender|SetTextRenderColor _editoronly_feedback_2 (LocalVariables|GetGreenColor)))
    (else
      _))
  (Rendering|Components|TextRender|SetText _editoronly_feedback (Utilities|Text|MakeLiteralText "INVTEXT(\"References are not set!\")"))
  (Rendering|Components|TextRender|SetTextRenderColor _editoronly_feedback (LocalVariables|GetRedColor)))
