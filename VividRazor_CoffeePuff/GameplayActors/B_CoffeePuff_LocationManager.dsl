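; CoffeePuff location manager that the spawn manager activates once as either a boss or non-boss location.
; After playlist data and references are ready, the server spawns either boss floor loot plus a gong encounter spawner given patrol points, or non-boss loot.
; An encounter state change triggers a one-time fight announcer GameplayEffect on players overlapping a box.
; The construction script displays editor feedback on reference setup, and missing runtime references log warnings.
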(event Custom|ActivateLocation
  (params (bBossLocation :type Boolean))
  ;[0 0.15 0.01476 0.5] ActivateLocation: Invoked to activate the location as either a boss location or not a boss location.
  ; Called from BP_CoffeePuff_SpawnManager.
  (Utilities|FlowControl|DoOnce false
    (:Completed
      (Fortnite|FortShippingLog self (Utilities|String|Append "ActivateLocation: Activating as boss location: " (Utilities|String|ToString(Boolean) bBossLocation)))
      (if (Networking|IsDedicatedServer)
        (Class|FortAsyncActionCurrentPlaylistReady|PlaylistDataReadyAsync
          (:OnReady
            (bind _attemptcount (Utilities|WaitUntilReferencesReady (ScalableFloat|GetValueasInteger (Variables|Hotfix|GetHotfix_FindReferencesMaxAttemptCount)))
              (:OutSuccess
                (Fortnite|FortShippingLog self (Utilities|String|Append "BeginPlay: All references found successfully at attempt count: " (Utilities|String|ToString(Integer) _attemptcount)))
                (if bBossLocation
                  (Class|BPoppySoapFloorLootItemSpawner|SpawnFloorLoot (Variables|ItemSpawners|Boss|GetBossSpawnMotoWrestleItemSpawner))
                  (bind _returnvalue (Building|SpawnBuildingGameplayActor "/VividRazor_CoffeePuff/GameplayActors/BGA_CoffeePuff_Spawner.BGA_CoffeePuff_Spawner_C" (Math|Transform|MakeTransform (Transformation|GetActorLocation (Variables|Gong|GetGongSpawnLocationActor)) (Transformation|GetActorRotation (Variables|Gong|GetGongSpawnLocationActor))) self true false))
                  (bind _asbga_coffee_puff_spawner (Utilities|Casting|CastToBGA_CoffeePuff_Spawner _returnvalue)
                    (:then
                      (bind _output_get (Variables|Transient|SetBossSpawner _asbga_coffee_puff_spawner))
                      (Class|BGACoffeePuffSpawner|SetPointProvider _output_get (Variables|Boss|GetBossPatrolPointProvider) :!class /VividRazor_CoffeePuff/GameplayActors/BGA_CoffeePuff_Spawner.BGA_CoffeePuff_Spawner_C)
                      (Fortnite|FortShippingLog self "BeginPlay: Successfully spawned Gong and set up boss point provider.")
                      (Default|BindEventtoOnEncounterStateChanged _asbga_coffee_puff_spawner (EventDispatchers|CreateEvent :!event HandleEncounterStateChanged)))
                    (:CastFailed
                      (Fortnite|FortShippingLog self "BeginPlay: Failed to spawn gong and failed to set up boss point provider." true)))
                  (else
                    (Class|BPoppySoapFloorLootItemSpawner|SpawnFloorLoot (Variables|ItemSpawners|NoBoss|GetNoBossCoffeePuffItemSpawner)))))
              (:OutFail
                (CallFunction|LogWarningforMissingReference)))))))))

(event Custom|AnnounceFight
  ;[0.1383 0.03516 0.15 0.5] AnnounceFight: When desired, play the fight VO on all nearby players.
  (if (Networking|IsDedicatedServer)
    (Utilities|FlowControl|DoOnce false
      (:Completed
        (Fortnite|FortShippingLog self "AnnounceFight: Invoked")
        (bind (_outactors _returnvalue) (Collision|BoxOverlapActors (Transformation|GetWorldTransform (Variables|Default|GetFightAnnouncerBox)) (Variables|/Script/Engine.BoxComponent|GetBoxExtent (Variables|Default|GetFightAnnouncerBox)) (Utilities|Array|MakeArray "ObjectTypeQuery3") "/Script/FortniteGame.FortPlayerPawn" :!class /Script/FortniteGame.FortKismetLibrary))
        (if _returnvalue
          (for _array_element _outactors
            (bind _asfort_player_pawn (Utilities|Casting|CastToFortPlayerPawn _array_element))
            (if (Pawn|IsPlayerControlled _asfort_player_pawn)
              (Fortnite|FortShippingLog self (Utilities|String|Append "AnnounceFight: Applying announcer GE to playerpawn: " (Utilities|GetDisplayName _asfort_player_pawn)))
              (Fortnite|ApplyGameplayEffectsToActor _asfort_player_pawn (Utilities|Array|MakeArray (Variables|Announcer|GetFightAnnouncerGameplayEffectClass)) (0.0 0.0 0.0) false 0 (Ability|GetFortAbilitySystemComponent _asfort_player_pawn) (GameplayEffects|MakeEffectContext (Ability|GetFortAbilitySystemComponent _asfort_player_pawn)))))
          (else
            (Fortnite|FortShippingLog self "AnnounceFight: BoxOverlap found no player pawns.")))))))

(event EventEndPlay
  (params (EndPlayReason :type EEndPlayReason%20Enum :default "Destroyed"))
  ;[0.15 0.09083 0 0.5] EndPlay: Clean up delegate.
  (Utilities|FlowControl|Sequence
    (:then_0
      (bind _bossspawner (Variables|Transient|GetBossSpawner)
        (:then
          (Default|UnbindEventfromOnEncounterStateChanged _bossspawner (EventDispatchers|CreateEvent :!event HandleEncounterStateChanged)))
        (:else)))
    (:then_1
      (|CallParentFunction EndPlayReason))))

(fn HandleEncounterStateChanged
  (params (EncounterState :type ECoffeePuffEncounterState%20Enum :default "NewEnumerator0"))
  ;[0.001016 0.1385 0.15 0.5] HandleEncounterStateChanged: When the counter has started, announce fight VO.
  (if (Networking|IsDedicatedServer)
    (if (Utilities|Enum|Equal(Enum) EncounterState "NewEnumerator1")
      (CallFunction|AnnounceFight)
      (bind _bossspawner (Variables|Transient|GetBossSpawner)
        (:then
          (Default|UnbindEventfromOnEncounterStateChanged _bossspawner (EventDispatchers|CreateEvent :!event HandleEncounterStateChanged)))
        (:else)))))

(fn IsBossSpawningEnabledAtLocation
  (return (ScalableFloat|GetValueasBool (Variables|Hotfix|GetHotfix_BossSpawningEnabled))))

(fn LogWarningForMissingReference
  ;[0.1456 0.000242 0.15 0.5] Log warning for any missing references.
  (Utilities|FlowControl|Sequence
    (:then_0
      (Variables|ItemSpawners|NoBoss|GetNoBossCoffeePuffItemSpawner
        (:then)
        (:else
          (Fortnite|FortShippingLog self "NoBossCoffeePuffItemSpawner: Count not find." true))))
    (:then_1
      (Variables|ItemSpawners|NoBoss|GetNoBossMotoWrestleItemSpawner
        (:then)
        (:else
          (Fortnite|FortShippingLog self "NoBossMotoWrestleItemSpawner: Count not find." true))))
    (:then_2
      (Variables|ItemSpawners|Boss|GetBossSpawnMotoWrestleItemSpawner
        (:then)
        (:else
          (Fortnite|FortShippingLog self "BossMotoWrestleItemSpawner: Count not find." true))))
    (:then_3
      (Variables|Gong|GetGongSpawnLocationActor
        (:then)
        (:else
          (Fortnite|FortShippingLog self "GongSpawnLocationActor: Count not find." true))))
    (:then_4
      (Variables|Boss|GetBossPatrolPointProvider
        (:then)
        (:else
          (Fortnite|FortShippingLog self "BossPatrolPointProvider: Count not find." true))))))

(fn ConstructionScript
  (bind _editoronly_feedback (Variables|Default|GetEDITORONLY_Feedback))
  (bind _editoronly_feedback_2 (Variables|Default|GetEDITORONLY_Feedback))
  (|CallParentFunction)
  (Variables|ItemSpawners|NoBoss|GetNoBossCoffeePuffItemSpawner
    (:then
      (Variables|ItemSpawners|NoBoss|GetNoBossMotoWrestleItemSpawner
        (:then
          (Variables|ItemSpawners|Boss|GetBossSpawnMotoWrestleItemSpawner
            (:then
              (Variables|Gong|GetGongSpawnLocationActor
                (:then
                  (Variables|Boss|GetBossPatrolPointProvider
                    (:then
                      (Rendering|Components|TextRender|SetText _editoronly_feedback_2 (Utilities|Text|MakeLiteralText "INVTEXT(\"References set!\")"))
                      (Rendering|Components|TextRender|SetTextRenderColor _editoronly_feedback_2 (LocalVariables|GetGreenColor)))
                    (:else
                      _)))
                (:else
                  _))
              (Rendering|Components|TextRender|SetText _editoronly_feedback (Utilities|Text|MakeLiteralText "INVTEXT(\"References are not set!\")"))
              (Rendering|Components|TextRender|SetTextRenderColor _editoronly_feedback (LocalVariables|GetRedColor)))
            (:else
              _)))
        (:else
          _)))
    (:else
      _)))
